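# Sound cue for a cluster bomb releasing its bomblets, classed as a medium explosive sound.
# A Random node picks one of six release wave variations with equal weight and is excluded from branch culling.
# A Modulator then applies slight pitch variation, and the cue plays at reduced volume.
# It borrows the bomb's throw-land attenuation and uses a release-duck concurrency set to manage overlapping plays.
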SoundCue /ClusterBombGameplay/Audio/ClusterBomb_Cluster_Release
root=Modulator_0
nodes [
 {id=Modulator_0, children [Random_0], properties {pitchMax=1.02, pitchMin=0.98}}
 {
  id=Random_0
  children [WavePlayer_0,WavePlayer_1,WavePlayer_2,WavePlayer_3,WavePlayer_4,WavePlayer_5]
  properties {bShouldExcludeFromBranchCulling=true, weights [1,1,1,1,1,1]}
 }
 {id=WavePlayer_0, properties {soundWaveAssetPtr=&/ClusterBomb_Cluster_Release_03}}
 {id=WavePlayer_1, properties {soundWaveAssetPtr=&/ClusterBomb_Cluster_Release_02}}
 {id=WavePlayer_2, properties {soundWaveAssetPtr=&/ClusterBomb_Cluster_Release_01}}
 {id=WavePlayer_3, properties {soundWaveAssetPtr=&/ClusterBomb_Cluster_Release_04}}
 {id=WavePlayer_4, properties {soundWaveAssetPtr=&/ClusterBomb_Cluster_Release_05}}
 {id=WavePlayer_5, properties {soundWaveAssetPtr=&/ClusterBomb_Cluster_Release_06}}
]
properties {
 attenuationSettings=/ClusterBombGameplay/Audio/Settings/ClusterBomb_ThrowLand
 maxDistance=20500
 soundClassObject=/Game/Sounds/FortSoundClassesAndModes/Fort_Explosives_Medium
 volumeMultiplier=0.4
 concurrencySet [/ClusterBombGameplay/Audio/Settings/ClusterBomb_ReleaseDuck]
}
