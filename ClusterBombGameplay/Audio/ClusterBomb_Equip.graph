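# Sound cue for equipping the cluster bomb, classed as weapon foley.
# A Random node picks one of three equally weighted equip waves and feeds a Modulator left at its default settings.
# It uses the shared weapon reload attenuation asset at a reduced volume.
# A concurrency override caps it at two instances, taking precedence over the assigned one-per-owner concurrency set.
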SoundCue /ClusterBombGameplay/Audio/ClusterBomb_Equip
root=Modulator_0
nodes [
 {id=Modulator_0, children [Random_0]}
 {id=Random_0, children [WavePlayer_0,WavePlayer_2,WavePlayer_3], properties {bShouldExcludeFromBranchCulling=true, weights [1,1,1]}}
 {id=WavePlayer_0, properties {soundWaveAssetPtr=&/ClusterBomb_Equip_01}}
 {id=WavePlayer_2, properties {soundWaveAssetPtr=&/ClusterBomb_Equip_02}}
 {id=WavePlayer_3, properties {soundWaveAssetPtr=&/ClusterBomb_Equip_03}}
]
properties {
 attenuationSettings=&Fort_Weapon_Reload_Att
 bOverrideConcurrency=true
 maxDistance=2250
 soundClassObject=&Fort_WeaponFoley
 volumeMultiplier=0.551
 attenuationOverrides {falloffDistance=1600}
 concurrencyOverrides {maxCount=2}
 concurrencySet [/Game/Sounds/Concurrency_Rules/OneMaximum_Owner_StopOldestRelease]
}
