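# Sound cue for the cluster bomb explosion, blending close and distant recordings by listener distance.
# A distance crossfade fades close sounds out as distant ones fade in over an overlapping range.
# Each layer picks randomly among five equally weighted waves, then passes through a modulator for pitch and volume variation.
# The cue uses dedicated attenuation, a medium explosives sound class, and release-duck concurrency sets including a water variant.
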SoundCue /ClusterBombGameplay/Audio/ClusterBomb_Explode_Cue
root=DistanceCrossFade_0
nodes [
 {
  id=DistanceCrossFade_0
  children [Modulator_8,Modulator_0]
  properties {crossFadeInput [
    {fadeInDistanceEnd=0, fadeInDistanceStart=0, fadeOutDistanceEnd=8000, fadeOutDistanceStart=3000, volume=1}
    {fadeInDistanceEnd=10000, fadeInDistanceStart=5000, fadeOutDistanceEnd=0, fadeOutDistanceStart=0, volume=1}
   ]}
 }
 {id=Modulator_0, children [Random_0], properties {pitchMax=1.1, volumeMax=1.2, volumeMin=1}}
 {id=Modulator_8, children [Random_2], properties {pitchMax=1.1, volumeMin=0.8}}
 {
  id=Random_0
  children [WavePlayer_1,WavePlayer_2,WavePlayer_3,WavePlayer_4,WavePlayer_5]
  properties {bShouldExcludeFromBranchCulling=true, weights [1,1,1,1,1]}
 }
 {
  id=Random_2
  children [WavePlayer_0, WavePlayer_10, WavePlayer_12, WavePlayer_13, WavePlayer_14]
  properties {bShouldExcludeFromBranchCulling=true, weights [1,1,1,1,1]}
 }
 {id=WavePlayer_0, properties {soundWaveAssetPtr=&/ClusterBomb_Explode_05}}
 {id=WavePlayer_1, properties {soundWaveAssetPtr=&/ClusterBomb_Explode_Distant_01}}
 {id=WavePlayer_10, properties {soundWaveAssetPtr=&/ClusterBomb_Explode_04}}
 {id=WavePlayer_12, properties {soundWaveAssetPtr=&/ClusterBomb_Explode_03}}
 {id=WavePlayer_13, properties {soundWaveAssetPtr=&/ClusterBomb_Explode_02}}
 {id=WavePlayer_14, properties {soundWaveAssetPtr=&/ClusterBomb_Explode_01}}
 {id=WavePlayer_2, properties {soundWaveAssetPtr=&/ClusterBomb_Explode_Distant_02}}
 {id=WavePlayer_3, properties {soundWaveAssetPtr=&/ClusterBomb_Explode_Distant_03}}
 {id=WavePlayer_4, properties {soundWaveAssetPtr=&/ClusterBomb_Explode_Distant_04}}
 {id=WavePlayer_5, properties {soundWaveAssetPtr=&/ClusterBomb_Explode_Distant_05}}
]
properties {
 attenuationSettings=/ClusterBombGameplay/Audio/Settings/ClusterBomb_Explode
 maxDistance=20500
 priority=2
 soundClassObject=/Game/Sounds/FortSoundClassesAndModes/Fort_Explosives_Medium
 volumeMultiplier=0.9
 concurrencyOverrides {maxCount=5, resolutionRule=StopOldest, volumeScale=0.8, volumeScaleAttackTime=0.1}
 concurrencySet [/ClusterBombGameplay/Audio/Settings/ClusterBomb_ReleaseDuck, /ClusterBombGameplay/Audio/Settings/ClusterBomb_ReleaseDuck_Water]
}
