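# Sound cue for a cluster bomb exploding in water.
# A distance crossfade blends a close layer, which fades out with range, into a distant layer that fades in farther away.
# Each layer randomly picks one of five wave variations and modulates its pitch and volume, with the distant layer boosted louder.
# It uses shared cluster bomb attenuation, a medium explosives sound class and a release-duck concurrency set.
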SoundCue /ClusterBombGameplay/Audio/ClusterBomb_Explode_Water
root=DistanceCrossFade_0
nodes [
 {
  id=DistanceCrossFade_0
  children [Modulator_8,Modulator_0]
  properties {crossFadeInput [
    {fadeInDistanceEnd=0, fadeInDistanceStart=0, fadeOutDistanceEnd=8000, fadeOutDistanceStart=4000, volume=1}
    {fadeInDistanceEnd=10000, fadeInDistanceStart=6000, fadeOutDistanceEnd=0, fadeOutDistanceStart=0, volume=1}
   ]}
 }
 {id=Modulator_0, children [Random_3], properties {pitchMax=1.1, volumeMax=2.2, volumeMin=2}}
 {id=Modulator_8, children [Random_1], properties {pitchMax=1.1, volumeMin=0.8}}
 {
  id=Random_1
  children [WavePlayer_6, WavePlayer_7, WavePlayer_8, WavePlayer_9, WavePlayer_11]
  properties {bShouldExcludeFromBranchCulling=true, weights [1,1,1,1,1]}
 }
 {
  id=Random_3
  children [WavePlayer_15, WavePlayer_16, WavePlayer_17, WavePlayer_18, WavePlayer_19]
  properties {bShouldExcludeFromBranchCulling=true, weights [1,1,1,1,1]}
 }
 {id=WavePlayer_11, properties {soundWaveAssetPtr=&/ClusterBomb_Explode_Water_05}}
 {id=WavePlayer_15, properties {soundWaveAssetPtr=&/ClusterBomb_Explode_Water_Distant_01}}
 {id=WavePlayer_16, properties {soundWaveAssetPtr=&/ClusterBomb_Explode_Water_Distant_02}}
 {id=WavePlayer_17, properties {soundWaveAssetPtr=&/ClusterBomb_Explode_Water_Distant_03}}
 {id=WavePlayer_18, properties {soundWaveAssetPtr=&/ClusterBomb_Explode_Water_Distant_04}}
 {id=WavePlayer_19, properties {soundWaveAssetPtr=&/ClusterBomb_Explode_Water_Distant_05}}
 {id=WavePlayer_6, properties {soundWaveAssetPtr=&/ClusterBomb_Explode_Water_01}}
 {id=WavePlayer_7, properties {soundWaveAssetPtr=&/ClusterBomb_Explode_Water_02}}
 {id=WavePlayer_8, properties {soundWaveAssetPtr=&/ClusterBomb_Explode_Water_03}}
 {id=WavePlayer_9, properties {soundWaveAssetPtr=&/ClusterBomb_Explode_Water_04}}
]
properties {
 attenuationSettings=/ClusterBombGameplay/Audio/Settings/ClusterBomb_Explode
 maxDistance=20500
 priority=3
 soundClassObject=/Game/Sounds/FortSoundClassesAndModes/Fort_Explosives_Medium
 volumeMultiplier=1
 concurrencyOverrides {maxCount=5, resolutionRule=StopOldest, volumeScale=0.8, volumeScaleAttackTime=0.1}
 concurrencySet [/ClusterBombGameplay/Audio/Settings/ClusterBomb_ReleaseDuck]
}
