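# Sound cue for a cluster bomb explosion's water splash, routed to the medium explosives sound class.
# A distance crossfade blends five random close variants, with randomized volume, into five random distant variants over overlapping ranges.
# It uses the cluster bomb explosion attenuation and a water release-duck concurrency set, alongside override values that cap instances and stop the oldest.
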SoundCue /ClusterBombGameplay/Audio/ClusterBomb_Explode_WaterSplash
root=DistanceCrossFade_0
nodes [
 {
  id=DistanceCrossFade_0
  children [Modulator_0,Random_1]
  properties {crossFadeInput [
    {fadeInDistanceEnd=0, fadeInDistanceStart=0, fadeOutDistanceEnd=8000, fadeOutDistanceStart=4000, volume=1}
    {fadeInDistanceEnd=10000, fadeInDistanceStart=6000, fadeOutDistanceEnd=0, fadeOutDistanceStart=0, volume=1}
   ]}
 }
 {id=Modulator_0, children [Random_0], properties {volumeMax=0.6, volumeMin=0.3}}
 {
  id=Random_0
  children [WavePlayer_0,WavePlayer_1,WavePlayer_2,WavePlayer_3,WavePlayer_4]
  properties {bShouldExcludeFromBranchCulling=true, weights [1,1,1,1,1]}
 }
 {id=Random_1, children [WavePlayer_5,WavePlayer_6,WavePlayer_7,WavePlayer_8,WavePlayer_9], properties {weights [1,1,1,1,1]}}
 {id=WavePlayer_0, properties {soundWaveAssetPtr=&/ClusterBomb_Explode_WaterSplash_01}}
 {id=WavePlayer_1, properties {soundWaveAssetPtr=&/ClusterBomb_Explode_WaterSplash_02}}
 {id=WavePlayer_2, properties {soundWaveAssetPtr=&/ClusterBomb_Explode_WaterSplash_03}}
 {id=WavePlayer_3, properties {soundWaveAssetPtr=&/ClusterBomb_Explode_WaterSplash_04}}
 {id=WavePlayer_4, properties {soundWaveAssetPtr=&/ClusterBomb_Explode_WaterSplash_05}}
 {id=WavePlayer_5, properties {soundWaveAssetPtr=&/ClusterBomb_Explode_WaterSplash_Distant_05}}
 {id=WavePlayer_6, properties {soundWaveAssetPtr=&/ClusterBomb_Explode_WaterSplash_Distant_04}}
 {id=WavePlayer_7, properties {soundWaveAssetPtr=&/ClusterBomb_Explode_WaterSplash_Distant_03}}
 {id=WavePlayer_8, properties {soundWaveAssetPtr=&/ClusterBomb_Explode_WaterSplash_Distant_02}}
 {id=WavePlayer_9, properties {soundWaveAssetPtr=&/ClusterBomb_Explode_WaterSplash_Distant_01}}
]
properties {
 attenuationSettings=/ClusterBombGameplay/Audio/Settings/ClusterBomb_Explode
 maxDistance=20500
 priority=3
 soundClassObject=/Game/Sounds/FortSoundClassesAndModes/Fort_Explosives_Medium
 volumeMultiplier=0.6
 concurrencyOverrides {maxCount=5, resolutionRule=StopOldest, volumeScale=0.8, volumeScaleAttackTime=0.1}
 concurrencySet [/ClusterBombGameplay/Audio/Settings/ClusterBomb_ReleaseDuck_Water]
}
