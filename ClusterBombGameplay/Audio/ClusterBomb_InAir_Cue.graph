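# Sound cue for a cluster bomb in flight, pairing a throw-start one-shot with a looping in-air sound.
# A mixer plays both layers together, with the throw start turned well down and the loop boosted.
# Each layer passes through a modulator and randomly picks one of two equally weighted waves.
# It uses the gadgets sound class, a dedicated in-air attenuation asset and a shared explosive-fuse concurrency rule.
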SoundCue /ClusterBombGameplay/Audio/ClusterBomb_InAir_Cue
root=Mixer_3
nodes [
 {id=Mixer_3, children [Modulator_2,Modulator_3], properties {inputVolume [0.2,2]}}
 {id=Modulator_2, children [Random_2]}
 {id=Modulator_3, children [Random_0]}
 {id=Random_0, children [WavePlayer_2,WavePlayer_4], properties {bShouldExcludeFromBranchCulling=true, weights [1,1]}}
 {id=Random_2, children [WavePlayer_5,WavePlayer_6], properties {bShouldExcludeFromBranchCulling=true, weights [1,1]}}
 {id=WavePlayer_2, properties {bLooping=true, soundWaveAssetPtr=&/ClusterBomb_Throw_InAir_Loop_01}}
 {id=WavePlayer_4, properties {bLooping=true, soundWaveAssetPtr=&/ClusterBomb_Throw_InAir_Loop_02}}
 {id=WavePlayer_5, properties {soundWaveAssetPtr=&/ClusterBomb_ThrowStart_02}}
 {id=WavePlayer_6, properties {soundWaveAssetPtr=&/ClusterBomb_ThrowStart_01}}
]
properties {
 attenuationSettings=/ClusterBombGameplay/Audio/Settings/ClusterBomb_InAir
 maxDistance=6800
 soundClassObject=&Fort_Gadgets
 volumeMultiplier=1
 attenuationOverrides {falloffDistance=5000, attenuationShapeExtents {x=1000}}
 concurrencyOverrides {bLimitToOwner=true, maxCount=3, resolutionRule=StopQuietest}
 concurrencySet [/Game/Sounds/Concurrency_Rules/Athena_Explosive_Fuse_Conc]
}
