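# ClusterBomb_Pickup is the sound cue for picking up the Cluster Bomb, classed as a loot-collect sound.
# A Modulator wraps a Random node that picks one of three equally weighted pickup waves and is excluded from branch culling.
# It uses shared weapon-pickup attenuation and a boosted volume.
# A concurrency override caps it at two simultaneous instances, stopping the oldest.
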SoundCue /ClusterBombGameplay/Audio/ClusterBomb_Pickup
root=Modulator_0
nodes [
 {id=Modulator_0, children [Random_1]}
 {id=Random_1, children [WavePlayer_2,WavePlayer_4,WavePlayer_5], properties {bShouldExcludeFromBranchCulling=true, weights [1,1,1]}}
 {id=WavePlayer_2, properties {soundWaveAssetPtr=&/ClusterBomb_Pickup_03}}
 {id=WavePlayer_4, properties {soundWaveAssetPtr=&/ClusterBomb_Pickup_02}}
 {id=WavePlayer_5, properties {soundWaveAssetPtr=&/ClusterBomb_Pickup_01}}
]
properties {
 attenuationSettings=&Fort_Weapon_Pickup_Att
 bOverrideConcurrency=true
 maxDistance=3500
 soundClassObject=&Fort_Loot_Collect
 volumeMultiplier=1.9
 concurrencyOverrides {maxCount=2, resolutionRule=StopOldest}
}
