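# Sound cue for the in-air sound of a cluster bomb's small bomblets, routed to the gadgets sound class.
# A Random node picks one of four equally weighted loop waves, and a Modulator adds slight upward pitch variation.
# An Enveloper at the root ramps pitch up over the first fraction of a second while keeping volume flat.
# Dedicated attenuation and concurrency assets govern spatial falloff and simultaneous instances.
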SoundCue /ClusterBombGameplay/Audio/ClusterBomb_SmallBomb_InAir
root=Enveloper_0
nodes [
 {
  id=Enveloper_0
  children [Modulator_0]
  properties {pitchCurve {editorCurveData {keys [{time=0}, <at 0.2497 1.1586>, <at 0.6 1.5>]}}, volumeCurve {editorCurveData {keys [<at 0 1>]}}}
 }
 {id=Modulator_0, children [Random_1], properties {pitchMax=1.1, pitchMin=1}}
 {
  id=Random_1
  children [WavePlayer_0,WavePlayer_1,WavePlayer_3,WavePlayer_7]
  properties {bShouldExcludeFromBranchCulling=true, weights [1,1,1,1]}
 }
 {id=WavePlayer_0, properties {soundWaveAssetPtr=&/ClusterBomb_SmallBomb_InAir_Loop_01}}
 {id=WavePlayer_1, properties {soundWaveAssetPtr=&/ClusterBomb_SmallBomb_InAir_Loop_02}}
 {id=WavePlayer_3, properties {soundWaveAssetPtr=&/ClusterBomb_SmallBomb_InAir_Loop_03}}
 {id=WavePlayer_7, properties {soundWaveAssetPtr=&/ClusterBomb_SmallBomb_InAir_Loop_04}}
]
properties {
 attenuationSettings=/ClusterBombGameplay/Audio/Settings/ClusterBomb_InAir
 maxDistance=6800
 soundClassObject=&Fort_Gadgets
 volumeMultiplier=0.7
 attenuationOverrides {falloffDistance=5000, attenuationShapeExtents {x=1000}}
 concurrencyOverrides {bLimitToOwner=true, maxCount=3, resolutionRule=StopQuietest}
 concurrencySet [/ClusterBombGameplay/Audio/Settings/ClusterBomb_SmallBomb_InAir]
}
