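# Sound cue for the cluster bomb's throw landing in water, playing a water variant of its warning beep.
# A short fixed delay feeds a modulator with slight volume variation, then an equally weighted random pick from three beep waves.
# The random node is excluded from branch culling.
# The cue uses the cluster bomb throw-land attenuation asset and the medium explosives sound class, with a slightly boosted volume.
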SoundCue /ClusterBombGameplay/Audio/ClusterBomb_ThrowLand_Water
root=Delay_1
nodes [
 {id=Delay_1, children [Modulator_1], properties {delayMax=0.1, delayMin=0.1}}
 {id=Modulator_1, children [Random_2], properties {volumeMin=0.9}}
 {id=Random_2, children [WavePlayer_7,WavePlayer_8,WavePlayer_9], properties {bShouldExcludeFromBranchCulling=true, weights [1,1,1]}}
 {id=WavePlayer_7, properties {soundWaveAssetPtr=&/ClusterBomb_WarningBeep_Water_03}}
 {id=WavePlayer_8, properties {soundWaveAssetPtr=&/ClusterBomb_WarningBeep_Water_02}}
 {id=WavePlayer_9, properties {soundWaveAssetPtr=&/ClusterBomb_WarningBeep_Water_01}}
]
properties {
 attenuationSettings=/ClusterBombGameplay/Audio/Settings/ClusterBomb_ThrowLand
 maxDistance=20500
 soundClassObject=/Game/Sounds/FortSoundClassesAndModes/Fort_Explosives_Medium
 volumeMultiplier=1.2
}
